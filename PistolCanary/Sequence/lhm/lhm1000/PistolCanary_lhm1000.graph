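# PistolCanary lhm1000 master Level Sequence, which assembles one cinematic shot from per-department subsequences at a 60 fps display rate.
# Seven sub tracks (animation, audio, camera, gameplay, lighting, techart, vfx) run in parallel over the same frame range.
# Its only binding, a FortStaticMeshActor named lhm1000_attachToPlayer, has an identity default transform and a spawn track defaulting to false.
# The name suggests a player-attachment anchor, though no attach track appears here.
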LevelSequence /PistolCanary/Sequence/lhm/lhm1000/PistolCanary_lhm1000
bindings [lhm1000_attachToPlayer {
  id=#1
  possessed_class=FortStaticMeshActor
  tracks [
   {type=MovieSceneSpawnTrack, sections [{type=MovieSceneSpawnSection, channels [None {type=MovieSceneScriptingBoolChannel, default=false}]}]}
   {
    type=MovieScene3DTransformTrack
    sections [{
      type=MovieScene3DTransformSection
      channels [
       Location.X {type=MovieSceneScriptingDoubleChannel, default=0}
       Location.Y {type=MovieSceneScriptingDoubleChannel, default=0}
       Location.Z {type=MovieSceneScriptingDoubleChannel, default=0}
       Rotation.X {type=MovieSceneScriptingDoubleChannel, default=0}
       Rotation.Y {type=MovieSceneScriptingDoubleChannel, default=0}
       Rotation.Z {type=MovieSceneScriptingDoubleChannel, default=0}
       Scale.X {type=MovieSceneScriptingDoubleChannel, default=1}
       Scale.Y {type=MovieSceneScriptingDoubleChannel, default=1}
       Scale.Z {type=MovieSceneScriptingDoubleChannel, default=1}
      ]
     }]
   }
  ]
 }]
display_rate {denominator=1, numerator=60}
tick_resolution {denominator=1, numerator=24000}
tracks [
 {type=MovieSceneSubTrack, display_name=animation, sections [{type=MovieSceneSubSection, end_frame=4318, start_frame=3600}]}
 {type=MovieSceneSubTrack, display_name=audio, sections [{type=MovieSceneSubSection, end_frame=4318, start_frame=3600}]}
 {type=MovieSceneSubTrack, display_name=camera, sections [{type=MovieSceneSubSection, end_frame=4318, start_frame=3600}]}
 {type=MovieSceneSubTrack, display_name=gameplay, sections [{type=MovieSceneSubSection, end_frame=4318, start_frame=3600}]}
 {type=MovieSceneSubTrack, display_name=lighting, sections [{type=MovieSceneSubSection, end_frame=4318, start_frame=3600}]}
 {type=MovieSceneSubTrack, display_name=techart, sections [{type=MovieSceneSubSection, end_frame=4318, start_frame=3600}]}
 {type=MovieSceneSubTrack, display_name=vfx, sections [{type=MovieSceneSubSection, end_frame=4318, start_frame=3600}]}
]
