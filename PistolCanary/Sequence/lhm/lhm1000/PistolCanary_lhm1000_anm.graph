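# Level Sequence for the animation of PistolCanary shot lhm1000, in which a superhero NPC (KalEl) flies in, hovers, then flies away.
# Two bindings of the NPC class follow similar keyed paths with a temporary scale-up; one layers overlapping animation sections and has keyed visibility.
# That binding and a spawned player mannequin, which plays one animation, carry attach tracks.
# A child Niagara component's LifetimeScale is keyed across the flight.
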LevelSequence /PistolCanary/Sequence/lhm/lhm1000/PistolCanary_lhm1000_anm
bindings [
 PistolCanary_lhm1000_KalEl {
  id=#1
  possessed_class=PistolCanary_CinematicNPC_Superhero_Main_C
  tracks [
   {
    type=MovieSceneVisibilityTrack
    sections [{type=MovieSceneVisibilitySection, channels [None {type=MovieSceneScriptingBoolChannel, default=true}]}]
   }
   {
    type=MovieSceneSkeletalAnimationTrack
    sections [
     {type=MovieSceneSkeletalAnimationSection, end_frame=3940, start_frame=3573}
     {type=MovieSceneSkeletalAnimationSection, end_frame=3964, start_frame=3940}
    ]
   }
   {
    type=MovieScene3DTransformTrack
    sections [{
      type=MovieScene3DTransformSection
      channels [
       Location.X {
        type=MovieSceneScriptingDoubleChannel
        default=12223.2996
        thinned_from=20
        keys [[3599.9975,20816.8399],[3600,-3198.8197],[3621.9975,-1976.9495],[3636.9975,-1161.5045],[3672,-81.3336],[3682,-41.8933],[3693,-19.0703],[3708,5.9426],[3730,16.9989],[3926,14.2256],[3929,14.2256],[3934,14.1707],[3936,14.0869],[3937,17.8529],[3938,48.0388],[3940,125.9077],[3948,251.1218],[3956,445.2026],[4066,2899.7485]]
       }
       Location.Y {
        type=MovieSceneScriptingDoubleChannel
        default=72556
        thinned_from=21
        keys [[3599.9975,77393.4936],[3600,10173.0719],[3621.9975,6544.3308],[3636.9975,4027.4133],[3672,655.1643],[3682,509.172],[3693,422.3298],[3708,353.8622],[3730,323.744],[3936,319.5484],[3937,310.6054],[3938,268.9324],[3940,170.1885],[3948,-248.7204],[3956,-799.7094],[3960,-1046.0926],[4025,-2490.7662]]
       }
       Location.Z {
        type=MovieSceneScriptingDoubleChannel
        default=5130.6223
        thinned_from=20
        keys [[3599.9975,14005.8375],[3600,8943.3396],[3621.9975,3559.1111],[3637,1064.5598],[3672,141.4266],[3682,151.7709],[3693,153.8522],[3708,156.2573],[3730,158.2848],[3934,161.2596],[3936,161.0706],[3937,172.5297],[3938,248.5021],[3940,374.9803],[3948,831.0652],[3956,922.124],[4018,2097.5301]]
       }
       Rotation.X {
        type=MovieSceneScriptingDoubleChannel
        default=-1.3184
        thinned_from=17
        keys [[3599.9975,27.2469],[3621.9975,27.2469],[3636.9975,26.9262],[3672,-0.2841],[3718,-1.3184],[3939,-1.3197],[3941,6.1653],[3943,17.3012],[3948,29.471],[3956,29.471]]
       }
       Rotation.Y {
        type=MovieSceneScriptingDoubleChannel
        default=2.6258
        thinned_from=18
        keys [[3599.9975,-10.0485],[3621.9975,-10.0485],[3636.9975,-3.1056],[3672,5.4069],[3718,2.6258],[3951.0025,2.6518],[3956,14.0946],[3960,22.2858],[3988,-6.6334]]
       }
       Rotation.Z {
        type=MovieSceneScriptingDoubleChannel
        default=16.0294
        thinned_from=19
        keys [[3599.9975,52.7023],[3600,-161.2977],[3621.9975,-161.2977],[3636.9975,-165.0143],[3672,-164.8596],[3718,-162.8863],[3935,-162.9102],[3941,-161.9791],[3951,-161.9869],[3956,-166.1871],[3960,-173.1213],[3988,-279.1771]]
       }
       Scale.X {type=MovieSceneScriptingDoubleChannel, default=1, thinned_from=10, keys [[3599.9975,1],[3606.9975,1.2],[3958,1.2],[3962,1.0019]]}
       Scale.Y {type=MovieSceneScriptingDoubleChannel, default=1, thinned_from=10, keys [[3599.9975,1],[3606.9975,1.2],[3958,1.2],[3962,1.0019]]}
       Scale.Z {type=MovieSceneScriptingDoubleChannel, default=1, thinned_from=10, keys [[3599.9975,1],[3606.9975,1.2],[3958,1.2],[3962,1.0019]]}
      ]
     }]
   }
  ]
 }
 player {
  id=#2
  possessed_class=FortPlayerMannequin
  tracks [
   {type=MovieSceneSpawnTrack, sections [{type=MovieSceneSpawnSection, channels [None {type=MovieSceneScriptingBoolChannel, default=true}]}]}
   {
    type=MovieScene3DTransformTrack
    sections [{
      type=MovieScene3DTransformSection
      channels [
       Location.X {type=MovieSceneScriptingDoubleChannel, default=120}
       Location.Y {type=MovieSceneScriptingDoubleChannel, default=-2}
       Location.Z {type=MovieSceneScriptingDoubleChannel, default=69.4}
       Rotation.X {type=MovieSceneScriptingDoubleChannel, default=0}
       Rotation.Y {type=MovieSceneScriptingDoubleChannel, default=0}
       Rotation.Z {type=MovieSceneScriptingDoubleChannel, default=0}
       Scale.X {type=MovieSceneScriptingDoubleChannel, default=1}
       Scale.Y {type=MovieSceneScriptingDoubleChannel, default=1}
       Scale.Z {type=MovieSceneScriptingDoubleChannel, default=1}
      ]
     }]
   }
   {type=MovieSceneSkeletalAnimationTrack, sections [{type=MovieSceneSkeletalAnimationSection, end_frame=4392, start_frame=3500}]}
   {type=MovieScene3DAttachTrack, sections [{type=MovieScene3DAttachSection, end_frame=12085, start_frame=3600}]}
  ]
 }
 PistolCanary_CinematicNPC_Superhero_Main {
  id=#3
  possessed_class=PistolCanary_CinematicNPC_Superhero_Main_C
  tracks [
   {type=MovieSceneSpawnTrack, sections [{type=MovieSceneSpawnSection, channels [None {type=MovieSceneScriptingBoolChannel, default=true}]}]}
   {
    type=MovieSceneSkeletalAnimationTrack
    sections [
     {type=MovieSceneSkeletalAnimationSection, end_frame=3668, start_frame=3500}
     {type=MovieSceneSkeletalAnimationSection, end_frame=3932, start_frame=3565}
     {type=MovieSceneSkeletalAnimationSection, end_frame=3743, start_frame=3657}
     {type=MovieSceneSkeletalAnimationSection, end_frame=3759, start_frame=3677}
     {type=MovieSceneSkeletalAnimationSection, end_frame=3871, start_frame=3752}
     {type=MovieSceneSkeletalAnimationSection, end_frame=3922, start_frame=3822}
     {type=MovieSceneSkeletalAnimationSection, end_frame=4007, start_frame=3893}
     {type=MovieSceneSkeletalAnimationSection, end_frame=3966, start_frame=3932}
     {type=MovieSceneSkeletalAnimationSection, end_frame=4085, start_frame=3992}
    ]
   }
   {
    type=MovieScene3DTransformTrack
    sections [{
      type=MovieScene3DTransformSection
      channels [
       Location.X {
        type=MovieSceneScriptingDoubleChannel
        default=12223.2996
        thinned_from=9
        keys [[3599.9975,20816.8399],[3600,-3198.8197],[3621.9975,-1976.9495],[3636.9975,-1161.5045],[3672,-44.6676],[3718,-0.2243],[3926,-2.9975],[3988,-94.0002]]
       }
       Location.Y {
        type=MovieSceneScriptingDoubleChannel
        default=72556
        thinned_from=10
        keys [[3599.9975,77393.4936],[3600,10173.0719],[3621.9975,6544.3308],[3636.9975,4027.4133],[3672,402.7631],[3717.9975,321.0564],[3926.0025,316.9448],[3978,-481.3076]]
       }
       Location.Z {
        type=MovieSceneScriptingDoubleChannel
        default=5130.6223
        thinned_from=9
        keys [[3599.9975,14005.8375],[3600,8943.3396],[3621.9975,3559.1111],[3636.9975,1010.0874],[3672,86.9542],[3718,103.8124],[3926.0025,106.7874],[3978,658.9877]]
       }
       Rotation.X {
        type=MovieSceneScriptingDoubleChannel
        default=-1.3184
        thinned_from=10
        keys [[3599.9975,27.2469],[3621.9975,27.2469],[3636.9975,26.9262],[3672,-0.2841],[3718,-1.3184],[3935,-1.3198],[3952,-1.4514],[3988,20.5613]]
       }
       Rotation.Y {
        type=MovieSceneScriptingDoubleChannel
        default=2.6258
        thinned_from=9
        keys [[3599.9975,-10.0485],[3621.9975,-10.0485],[3636.9975,-3.1056],[3672,2.7722],[3718,2.6258],[3934.0025,2.6526],[3951.0025,25.639],[3988,-6.6334]]
       }
       Rotation.Z {
        type=MovieSceneScriptingDoubleChannel
        default=16.0294
        thinned_from=9
        keys [[3599.9975,52.7023],[3600,-161.2977],[3621.9975,-161.2977],[3636.9975,-164.2915],[3672,-172.9302],[3952,-173.0319],[3988,-279.1771]]
       }
       Scale.X {type=MovieSceneScriptingDoubleChannel, default=1, keys [[3599.9975,1],[3606.9975,1.2],[3935,1.2],[3952,1.0019]]}
       Scale.Y {type=MovieSceneScriptingDoubleChannel, default=1, keys [[3599.9975,1],[3606.9975,1.2],[3935,1.2],[3952,1.0019]]}
       Scale.Z {type=MovieSceneScriptingDoubleChannel, default=1, keys [[3599.9975,1],[3606.9975,1.2],[3935,1.2],[3952,1.0019]]}
      ]
     }]
   }
   {
    type=MovieSceneVisibilityTrack
    sections [{
      type=MovieSceneVisibilitySection
      channels [None {type=MovieSceneScriptingBoolChannel, default=false, keys [[3899.9975,true],[4411.9975,false]]}]
     }]
   }
   {type=MovieScene3DAttachTrack, sections [{type=MovieScene3DAttachSection, end_frame=12085, start_frame=3600}]}
  ]
 }
 Niagara {
  id=#4
  parent_id=#3
  possessed_class=NiagaraComponent
  tracks [{
    type=MovieSceneNiagaraFloatParameterTrack
    display_name=LifetimeScale
    sections [{
      type=MovieSceneFloatSection
      channels [None {type=MovieSceneScriptingFloatChannel, default=0, keys [[3600,0],[3617,10],[3674,10],[3677,0],[3719,5],[4099.9975,0]]}]
     }]
   }]
 }
]
display_rate {denominator=1, numerator=60}
tick_resolution {denominator=1, numerator=24000}
tracks []
