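# Audio Level Sequence for cinematic shot lhm1000 in PistolCanary, running at 60 fps.
# A single MainHeroMovement audio track holds one section with volume and pitch at unity and an attach-actor channel; no sound asset is shown.
# The KalEl superhero cinematic NPC is bound as a possessable but carries no tracks of its own.
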LevelSequence /PistolCanary/Sequence/lhm/lhm1000/PistolCanary_lhm1000_aud
bindings [PistolCanary_lhm1000_KalEl {id=#1, possessed_class=PistolCanary_CinematicNPC_Superhero_Main_C, tracks []}]
display_rate {denominator=1, numerator=60}
tick_resolution {denominator=1, numerator=24000}
tracks [{
  type=MovieSceneAudioTrack
  display_name=MainHeroMovement
  sections [{
    type=MovieSceneAudioSection
    end_frame=4012
    start_frame=3600
    channels [
     Volume {type=MovieSceneScriptingFloatChannel, default=1}
     Pitch {type=MovieSceneScriptingFloatChannel, default=1}
     AttachActor {type=MovieSceneScriptingActorReferenceChannel}
    ]
   }]
 }]
