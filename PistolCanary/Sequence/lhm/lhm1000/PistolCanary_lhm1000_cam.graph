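# Camera Level Sequence for the PistolCanary lhm1000 shot, with a camera cut track and two camera modifier events.
# The main cine camera animates position, rotation and focal length, layering five template sequences with keyed rotation scales that fade in and out.
# A camera force Blueprint sets a shot-specific shake and small force feedback, with event triggers for lobby and radius variants.
# A second cine camera previews the shake through its own template animation.
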LevelSequence /PistolCanary/Sequence/lhm/lhm1000/PistolCanary_lhm1000_cam
bindings [
 PistolCanary_cameraForce_BP {
  id=#1
  possessed_class=Events_Shared_cameraForce_BP_C
  tracks [
   {
    type=MovieSceneObjectPropertyTrack
    display_name=Force Effect
    sections [{
      type=MovieSceneObjectPropertySection
      channels [None {
        type=MovieSceneScriptingObjectPathChannel
        default="<Object '&Events_Shared_medium_FF' (0x8509a2598b0) Class 'ForceFeedbackEffect'>"
        const ["<Object '&Events_Shared_small_FF' (0x84c3c703480) Class 'ForceFeedbackEffect'>", 1]
       }]
     }]
   }
   {
    type=MovieSceneObjectPropertyTrack
    display_name=Cam Shake Base
    sections [{
      type=MovieSceneObjectPropertySection
      channels [None {
        type=MovieSceneScriptingObjectPathChannel
        default="<Object '&Events_Shared_00small_csb_C' (0x84c3c6a8a00) Class 'BlueprintGeneratedClass'>"
        const ["<Object '/PistolCanary/Camera/PistolCanary_lhm1000_csb.PistolCanary_lhm1000_csb_C' (0x852f8509000) Class 'BlueprintGeneratedClass'>", 1]
       }]
     }]
   }
   {
    type=MovieSceneEventTrack
    display_name=Cam Shake Radius Event Trigger
    sections [{type=MovieSceneEventTriggerSection, channels [None {type=MovieSceneScriptingEventChannel}]}]
   }
   {type=MovieSceneSpawnTrack, sections [{type=MovieSceneSpawnSection, channels [None {type=MovieSceneScriptingBoolChannel, default=true}]}]}
   {
    type=MovieSceneEventTrack
    display_name=Force Feedback Radius Event Trigger
    sections [{type=MovieSceneEventTriggerSection, channels [None {type=MovieSceneScriptingEventChannel}]}]
   }
   {
    type=MovieScene3DTransformTrack
    sections [{
      type=MovieScene3DTransformSection
      channels [
       Location.X {type=MovieSceneScriptingDoubleChannel, default=7084}
       Location.Y {type=MovieSceneScriptingDoubleChannel, default=1572}
       Location.Z {type=MovieSceneScriptingDoubleChannel, default=0}
       Rotation.X {type=MovieSceneScriptingDoubleChannel, default=0}
       Rotation.Y {type=MovieSceneScriptingDoubleChannel, default=0}
       Rotation.Z {type=MovieSceneScriptingDoubleChannel, default=0}
       Scale.X {type=MovieSceneScriptingDoubleChannel, default=1}
       Scale.Y {type=MovieSceneScriptingDoubleChannel, default=1}
       Scale.Z {type=MovieSceneScriptingDoubleChannel, default=1}
      ]
     }]
   }
   {
    type=MovieSceneIntegerTrack
    sections [{type=MovieSceneIntegerSection, channels [None {type=MovieSceneScriptingIntegerChannel, default=2, const [0,1]}]}]
   }
   {
    type=MovieSceneEventTrack
    display_name=Cam Shake Lobby Event Trigger
    sections [{
      type=MovieSceneEventTriggerSection
      channels [None {type=MovieSceneScriptingEventChannel, const ["<Struct 'MovieSceneEvent' (0x852f46aee20) {}>", 1]}]
     }]
   }
   {
    type=MovieSceneEventTrack
    display_name=Force Feedback Lobby Event Trigger
    sections [{
      type=MovieSceneEventTriggerSection
      channels [None {type=MovieSceneScriptingEventChannel, const ["<Struct 'MovieSceneEvent' (0x852f46aea60) {}>", 1]}]
     }]
   }
  ]
 }
 CameraComponent {
  id=#2
  parent_id=#3
  possessed_class=CineCameraComponent
  tracks [
   {
    type=MovieSceneFloatTrack
    display_name=Current Focal Length
    sections [{
      type=MovieSceneFloatSection
      channels [None {type=MovieSceneScriptingFloatChannel, default=0, keys [[3654,0],[3687,10],[3881.9975,10],[3914.9975,0]]}]
     }]
   }
   {
    type=MovieSceneFloatTrack
    display_name="Manual Focus Distance (Focus Settings)"
    sections [{type=MovieSceneFloatSection, channels [None {type=MovieSceneScriptingFloatChannel, default=427.0862}]}]
   }
   {
    type=MovieSceneFloatTrack
    display_name=Current Aperture
    sections [{type=MovieSceneFloatSection, channels [None {type=MovieSceneScriptingFloatChannel, default=2.8}]}]
   }
  ]
 }
 cameraShake_preview {
  id=#3
  possessed_class=CineCameraActor
  tracks [
   {type=MovieSceneSpawnTrack, sections [{type=MovieSceneSpawnSection, channels [None {type=MovieSceneScriptingBoolChannel, default=true}]}]}
   {type=TemplateSequenceTrack, display_name=Template Animation, sections [{type=TemplateSequenceSection, end_frame=3963, start_frame=3600}]}
   {
    type=MovieScene3DTransformTrack
    sections [{
      type=MovieScene3DTransformSection
      channels [
       Location.X {type=MovieSceneScriptingDoubleChannel, default=0}
       Location.Y {type=MovieSceneScriptingDoubleChannel, default=0}
       Location.Z {type=MovieSceneScriptingDoubleChannel, default=0}
       Rotation.X {type=MovieSceneScriptingDoubleChannel, default=0}
       Rotation.Y {type=MovieSceneScriptingDoubleChannel, default=0}
       Rotation.Z {type=MovieSceneScriptingDoubleChannel, default=0}
       Scale.X {type=MovieSceneScriptingDoubleChannel, default=1}
       Scale.Y {type=MovieSceneScriptingDoubleChannel, default=1}
       Scale.Z {type=MovieSceneScriptingDoubleChannel, default=1}
      ]
     }]
   }
  ]
 }
 camera_lhm1000 {
  id=#4
  possessed_class=CineCameraActor
  tracks [
   {
    type=MovieScene3DTransformTrack
    sections [{
      type=MovieScene3DTransformSection
      channels [
       Location.X {type=MovieSceneScriptingDoubleChannel, default=50, keys [[3600,50],[3687,114.1075]]}
       Location.Y {type=MovieSceneScriptingDoubleChannel, default=-260, keys [[3600,-260],[3688,-70.5222]]}
       Location.Z {type=MovieSceneScriptingDoubleChannel, default=225, keys [[3600,225],[3691,252.7451]]}
       Rotation.X {type=MovieSceneScriptingDoubleChannel, default=0}
       Rotation.Y {type=MovieSceneScriptingDoubleChannel, default=3, keys [[3247,0],[3600,3],[3687,0]]}
       Rotation.Z {type=MovieSceneScriptingDoubleChannel, default=96.2, keys [[3600,90],[3687,96.2]]}
       Scale.X {type=MovieSceneScriptingDoubleChannel, default=1}
       Scale.Y {type=MovieSceneScriptingDoubleChannel, default=1}
       Scale.Z {type=MovieSceneScriptingDoubleChannel, default=1}
      ]
     }]
   }
   {type=MovieSceneSpawnTrack, sections [{type=MovieSceneSpawnSection, channels [None {type=MovieSceneScriptingBoolChannel, default=true}]}]}
   {
    type=TemplateSequenceTrack
    display_name=Template Animation
    sections [
     {
      type=TemplateSequenceSection
      end_frame=5534
      start_frame=3291
      channels ["Transform[Rotation]" {type=MovieSceneScriptingFloatChannel, default=0.1, keys [[3600,0],[3660,0.1],[3912,0.1],[3962,0]]}]
     }
     {
      type=TemplateSequenceSection
      end_frame=4004
      start_frame=3914
      channels ["Transform[Rotation]" {type=MovieSceneScriptingFloatChannel, default=0, keys [[3914,0],[3936,0.21],[3962,0]]}]
     }
     {
      type=TemplateSequenceSection
      end_frame=4004
      start_frame=3921
      channels ["Transform[Rotation]" {type=MovieSceneScriptingFloatChannel, default=0, keys [[3921,0],[3936,0.23],[3962,0]]}]
     }
     {
      type=TemplateSequenceSection
      end_frame=3985
      start_frame=3928
      channels ["Transform[Rotation]" {type=MovieSceneScriptingFloatChannel, default=1, keys [[3928,0],[3936,1],[3962,0]]}]
     }
     {
      type=TemplateSequenceSection
      end_frame=3726
      start_frame=3617
      channels ["Transform[Rotation]" {type=MovieSceneScriptingFloatChannel, default=0, keys [[3617,0],[3653,0.08],[3695.5825,0.08],[3724.31,0]]}]
     }
    ]
   }
  ]
 }
 CameraComponent {
  id=#5
  parent_id=#4
  possessed_class=CineCameraComponent
  tracks [
   {
    type=MovieSceneFloatTrack
    display_name=Current Focal Length
    sections [{
      type=MovieSceneFloatSection
      channels [None {type=MovieSceneScriptingFloatChannel, default=18, keys [[3615,18],[3648,21],[3927,21],[3960,14]]}]
     }]
   }
   {
    type=MovieSceneFloatTrack
    display_name="Manual Focus Distance (Focus Settings)"
    sections [{type=MovieSceneFloatSection, channels [None {type=MovieSceneScriptingFloatChannel, default=427.0862}]}]
   }
   {
    type=MovieSceneFloatTrack
    display_name=Current Aperture
    sections [{type=MovieSceneFloatSection, channels [None {type=MovieSceneScriptingFloatChannel, default=2.8}]}]
   }
  ]
 }
]
display_rate {denominator=1, numerator=60}
tick_resolution {denominator=1, numerator=24000}
tracks [
 {
  type=MovieSceneEventTrack
  display_name=Camera Modifier Event
  sections [{
    type=MovieSceneEventTriggerSection
    channels [None {
      type=MovieSceneScriptingEventChannel
      keys [[3658, "<Struct 'MovieSceneEvent' (0x852f46af640) {}>"], [3953, "<Struct 'MovieSceneEvent' (0x852f46af500) {}>"]]
     }]
   }]
 }
 {type=MovieSceneCameraCutTrack, sections [{type=MovieSceneCameraCutSection, end_frame=3963, start_frame=3245}]}
]
