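# Level Sequence for the gam layer of PistolCanary shot lhm1000, handling the KalEl superhero NPC and focus point actors with no camera or animation tracks.
# KalEl is held visible and fires two event triggers later in the shot.
# A focus point spawns for the shot window, attaches with an identity transform, and fires two events shortly after spawning.
# A persistent focus actor is attached over the same range; the attach targets are not shown.
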LevelSequence /PistolCanary/Sequence/lhm/lhm1000/PistolCanary_lhm1000_gam
bindings [
 KalEl {
  id=#1
  possessed_class=PistolCanary_CinematicNPC_Superhero_Main_C
  tracks [
   {
    type=MovieSceneBindingLifetimeTrack
    display_name=Binding Lifetime
    sections [
     {type=MovieSceneBindingLifetimeSection, end_frame=-19291}
     {type=MovieSceneBindingLifetimeSection, start_frame=-19231}
     {type=MovieSceneBindingLifetimeSection, end_frame=-19231, start_frame=-19291}
    ]
   }
   {
    type=MovieSceneEventTrack
    sections [{
      type=MovieSceneEventTriggerSection
      channels [None {
        type=MovieSceneScriptingEventChannel
        keys [[3647.635, "<Struct 'MovieSceneEvent' (0x852f46af640) {}>"], [3923.0325, "<Struct 'MovieSceneEvent' (0x852f46af5a0) {}>"]]
       }]
     }]
   }
   {
    type=MovieSceneVisibilityTrack
    sections [{type=MovieSceneVisibilitySection, channels [None {type=MovieSceneScriptingBoolChannel, default=true, const [true,1]}]}]
   }
  ]
 }
 BP_PistolCanary_FocusPoint {
  id=#2
  possessed_class=BP_PistolCanary_FocusPoint_C
  tracks [
   {
    type=MovieSceneSpawnTrack
    sections [{type=MovieSceneSpawnSection, channels [None {type=MovieSceneScriptingBoolChannel, default=false, keys [[3614.4875,true],[3949.3525,false]]}]}]
   }
   {
    type=MovieScene3DTransformTrack
    sections [{
      type=MovieScene3DTransformSection
      channels [
       Location.X {type=MovieSceneScriptingDoubleChannel, default=0}
       Location.Y {type=MovieSceneScriptingDoubleChannel, default=0}
       Location.Z {type=MovieSceneScriptingDoubleChannel, default=0}
       Rotation.X {type=MovieSceneScriptingDoubleChannel, default=0}
       Rotation.Y {type=MovieSceneScriptingDoubleChannel, default=0}
       Rotation.Z {type=MovieSceneScriptingDoubleChannel, default=0}
       Scale.X {type=MovieSceneScriptingDoubleChannel, default=1}
       Scale.Y {type=MovieSceneScriptingDoubleChannel, default=1}
       Scale.Z {type=MovieSceneScriptingDoubleChannel, default=1}
      ]
     }]
   }
   {type=MovieScene3DAttachTrack, sections [{type=MovieScene3DAttachSection, end_frame=3965, start_frame=3600}]}
   {
    type=MovieSceneEventTrack
    sections [{
      type=MovieSceneEventTriggerSection
      channels [None {
        type=MovieSceneScriptingEventChannel
        keys [[3619.8175, "<Struct 'MovieSceneEvent' (0x852f46af500) {}>"], [3634.3575, "<Struct 'MovieSceneEvent' (0x852f46af000) {}>"]]
       }]
     }]
   }
  ]
 }
 BP_PistolCanary_Focus_Persistent {
  id=#3
  possessed_class=BP_PistolCanary_Focus_Persistent_C
  tracks [
   {type=MovieScene3DAttachTrack, sections [{type=MovieScene3DAttachSection, end_frame=3965, start_frame=3600}]}
   {type=MovieSceneBindingLifetimeTrack, display_name=Binding Lifetime, sections [{type=MovieSceneBindingLifetimeSection}]}
  ]
 }
]
display_rate {denominator=1, numerator=60}
tick_resolution {denominator=1, numerator=24000}
tracks []
