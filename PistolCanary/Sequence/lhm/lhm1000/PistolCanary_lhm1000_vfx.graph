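# VFX Level Sequence for the PistolCanary lhm1000 cinematic, timing Superman flare and launch effects around the Kal-El superhero character.
# Spawn tracks switch on a flare Niagara actor with a warm spotlight, then a launch Niagara actor later as the flare ends.
# Keyed flare Opacity and light Intensity fade in and out together.
# Attach tracks bind the effects and light component, while Kal-El's Niagara component gets a ContrailWidth parameter track.
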LevelSequence /PistolCanary/Sequence/lhm/lhm1000/PistolCanary_lhm1000_vfx
bindings [
 LightComponent0 {
  id=#1
  parent_id=#2
  possessed_class=SpotLightComponent
  tracks [
   {
    type=MovieSceneFloatTrack
    display_name=Intensity
    sections [{
      type=MovieSceneFloatSection
      channels [None {type=MovieSceneScriptingFloatChannel, default=50, keys [[3939.135,0],[3985.3325,50],[4185,50],[4234.53,0]]}]
     }]
   }
   {
    type=MovieSceneColorTrack
    display_name=Light Color
    sections [{
      type=MovieSceneColorSection
      channels [
       Color.R {type=MovieSceneScriptingFloatChannel, default=1}
       Color.G {type=MovieSceneScriptingFloatChannel, default=0.4969}
       Color.B {type=MovieSceneScriptingFloatChannel, default=0.1812}
       Color.A {type=MovieSceneScriptingFloatChannel, default=1}
      ]
     }]
   }
   {type=MovieScene3DAttachTrack, sections [{type=MovieScene3DAttachSection, end_frame=12126, start_frame=3931}]}
  ]
 }
 NiagaraComponent0 {
  id=#3
  parent_id=#4
  possessed_class=NiagaraComponent
  tracks [{
    type=MovieSceneNiagaraFloatParameterTrack
    display_name=Opacity
    sections [{
      type=MovieSceneFloatSection
      channels [None {
        type=MovieSceneScriptingFloatChannel
        default=0.5
        keys [[3939.135,0],[3968.6025,1],[3972.9475,0.5],[3979.49,1],[4183,1],[4188.8425,0.5],[4193.7575,0.5],[4203.3125,0]]
       }]
     }]
   }]
 }
 SpotLight {
  id=#2
  possessed_class=SpotLight
  tracks [
   {
    type=MovieSceneSpawnTrack
    sections [{type=MovieSceneSpawnSection, channels [None {type=MovieSceneScriptingBoolChannel, default=false, keys [[3851.9275,false],[3939.135,true]]}]}]
   }
   {
    type=MovieScene3DTransformTrack
    sections [{
      type=MovieScene3DTransformSection
      channels [
       Location.X {type=MovieSceneScriptingDoubleChannel, default=733.8392}
       Location.Y {type=MovieSceneScriptingDoubleChannel, default=655.56}
       Location.Z {type=MovieSceneScriptingDoubleChannel, default=62.1607}
       Rotation.X {type=MovieSceneScriptingDoubleChannel, default=94.3907}
       Rotation.Y {type=MovieSceneScriptingDoubleChannel, default=-5.4448}
       Rotation.Z {type=MovieSceneScriptingDoubleChannel, default=-120.8254}
       Scale.X {type=MovieSceneScriptingDoubleChannel, default=1}
       Scale.Y {type=MovieSceneScriptingDoubleChannel, default=1}
       Scale.Z {type=MovieSceneScriptingDoubleChannel, default=1}
      ]
     }]
   }
  ]
 }
 Niagara {
  id=#5
  parent_id=#6
  possessed_class=NiagaraComponent
  tracks [{
    type=MovieSceneNiagaraFloatParameterTrack
    display_name=ContrailWidth
    sections [{type=MovieSceneFloatSection, channels [None {type=MovieSceneScriptingFloatChannel, default=0.5, const [0.5,1]}]}]
   }]
 }
 NS_PistolCanary_Superman_Launch_Local {
  id=#7
  possessed_class=NiagaraActor
  tracks [
   {
    type=MovieSceneSpawnTrack
    sections [{type=MovieSceneSpawnSection, channels [None {type=MovieSceneScriptingBoolChannel, default=false, keys [[4006.1375,false],[4200,true]]}]}]
   }
   {
    type=MovieScene3DTransformTrack
    sections [{
      type=MovieScene3DTransformSection
      channels [
       Location.X {type=MovieSceneScriptingDoubleChannel, default=387.943}
       Location.Y {type=MovieSceneScriptingDoubleChannel, default=71.403}
       Location.Z {type=MovieSceneScriptingDoubleChannel, default=16.3496}
       Rotation.X {type=MovieSceneScriptingDoubleChannel, default=-10.6664}
       Rotation.Y {type=MovieSceneScriptingDoubleChannel, default=23.9106}
       Rotation.Z {type=MovieSceneScriptingDoubleChannel, default=14.9735}
       Scale.X {type=MovieSceneScriptingDoubleChannel, default=1}
       Scale.Y {type=MovieSceneScriptingDoubleChannel, default=1}
       Scale.Z {type=MovieSceneScriptingDoubleChannel, default=1}
      ]
     }]
   }
   {type=MovieScene3DAttachTrack, sections [{type=MovieScene3DAttachSection, end_frame=12085, start_frame=4169}]}
  ]
 }
 PistolCanary_lhm1000_KalEl {id=#6, possessed_class=PistolCanary_CinematicNPC_Superhero_Main_C, tracks []}
 NS_PistolCanary_Superman_Flare {
  id=#4
  possessed_class=NiagaraActor
  tracks [
   {
    type=MovieSceneSpawnTrack
    sections [{
      type=MovieSceneSpawnSection
      channels [None {type=MovieSceneScriptingBoolChannel, default=false, keys [[3851.9275,false],[3939.135,true],[4216,false]]}]
     }]
   }
   {
    type=MovieScene3DTransformTrack
    sections [{
      type=MovieScene3DTransformSection
      channels [
       Location.X {type=MovieSceneScriptingDoubleChannel, default=0}
       Location.Y {type=MovieSceneScriptingDoubleChannel, default=0}
       Location.Z {type=MovieSceneScriptingDoubleChannel, default=0}
       Rotation.X {type=MovieSceneScriptingDoubleChannel, default=0}
       Rotation.Y {type=MovieSceneScriptingDoubleChannel, default=0}
       Rotation.Z {type=MovieSceneScriptingDoubleChannel, default=0}
       Scale.X {type=MovieSceneScriptingDoubleChannel, default=1}
       Scale.Y {type=MovieSceneScriptingDoubleChannel, default=1}
       Scale.Z {type=MovieSceneScriptingDoubleChannel, default=1}
      ]
     }]
   }
   {type=MovieScene3DAttachTrack, sections [{type=MovieScene3DAttachSection, end_frame=12126, start_frame=3931}]}
  ]
 }
]
display_rate {denominator=1, numerator=60}
tick_resolution {denominator=1, numerator=24000}
tracks []
